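# Stereo MetaSound source for an ambient music bed, filed under the multi-product front end music.
# On play, it gets a wave from the Ambient Bed Array (default holds one ethereal bed) and starts a quad Wave Player.
# The player's On Play triggers a short MSP_Fade envelope that scales only the front left and right channels into the stereo output.
# The Stop trigger connects straight to OnFinished, with no fade-out.
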MetaSoundSource /Game/Sounds/Fort_Music/Music_MP_FrontEnd/wav/MS_MultiProduct_Amb_Bed
inputs [
 Ambient Bed Array {type=WaveAsset:Array}
 Stop {type=Trigger}
 UE.Source.OnPlay {type=Trigger}
]
interfaces [UE.OutputFormat.Stereo:1.0, UE.Source.OneShot:1.0, UE.Source:1.1]
outputs [
 "UE.OutputFormat.Stereo.Audio:0" {type=Audio}
 "UE.OutputFormat.Stereo.Audio:1" {type=Audio}
 UE.Source.OneShot.OnFinished {type=Trigger}
]
pages [Default {
  input_defaults {Stop=null, UE.Source.OnPlay=null, Ambient Bed Array [&/MP_Amb_Bed_Etherial_lp]}
  nodes [
   WavePlayer_0 = UE.Wave Player:Quad(
    Wave Asset=Array.Get:WaveAsset:Array(Array=^Ambient Bed Array, Trigger=^UE.Source.OnPlay).Element
    Play=UE.Reroute(Value=^UE.Source.OnPlay).Value
   ) {also {defaults {Loop="<MetaSoundBuilderResult.SUCCEEDED: 0>"}}}
   n1 = &MSP_Fade(In=@WavePlayer_0.On Play) {also {defaults {Fade Time=0.2}}}
   Multiply_0 = UE.Multiply:Audio by Float(AdditionalOperands=@n1.Value Out, PrimaryOperand=@WavePlayer_0.Out Front Right)
   Multiply_1 = UE.Multiply:Audio by Float(AdditionalOperands=@n1.Value Out, PrimaryOperand=@WavePlayer_0.Out Front Left)
  ]
  links [
   {dst="@UE.Source.OneShot.OnFinished", src="@Stop"}
   {dst="@UE.OutputFormat.Stereo.Audio:1", src=Multiply_0.Out}
   {dst="@UE.OutputFormat.Stereo.Audio:0", src=Multiply_1.Out}
  ]
 }]
properties {assetImportData=&self:AssetImportData, numChannels=2, outputFormat=Stereo}
